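# Anim Blueprint for the FluidWedge backpack, adding secondary motion to its front_stop, gun and ammo bones.
# The pack bone is offset by spine bend measured on the owner mesh, then cached to feed three tuned AnimDynamics chains.
# Bool blends select the front-end, skydive/parachute or default chain.
# A parameter-asset gravity function sets GravityOverride and sim-space variables, but the AnimGraph reads neither, using fixed gravity values.
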AnimBlueprint /BRCosmetics/Accessories/FORT_Backpacks/Backpack_M_MED_FluidWedge/Meshes/M_MED_FluidWedge_Pack_AnimBP
target_skeleton=&/M_MED_FluidWedge_Pack_Skeleton
graphs [
 AnimGraph {nodes [
   n1 = Anim Dynamics - Bone: front_stop_1!(ComponentPose=Local To Component(LocalPose=Use cached pose 'Sky/level'()), Alpha=1, GravityOverride:Vector="0,0,-650", ExternalForce="0,0,0") -> (Pose) {
    angularDampingOverride=0.9999
    angularSpringConstant=200
    bAngularSpring=true
    bOverrideAngularDamping=true
    bUseGravityOverride=true
    simulationSpace=BoneRelative
    boundBone {boneName=front_stop_1}
    componentAppliedLinearAccClamp=<4000 4000 4000>
    componentLinearAccScale=<0.7 0.4 0.26>
    componentLinearVelScale=<0.3 0.3 0.3>
    gravityOverride {z=-650}
    physicsBodyDefinitions [{
      boundBone {boneName=front_stop_1}
      boxExtents=<8 8 8>
      constraintSetup {angularLimitsMax=<0 5 90>, angularLimitsMin=<0 -5 0>, angularTarget=<1 0 0>}
      localJointOffset=<0.8 0 0>
     }]
    relativeSpaceBone {boneName=root}
   }
   n2 = Anim Dynamics - Bone: front_stop_2!(ComponentPose=@n1, Alpha=1, GravityOverride:Vector="0,0,-650", ExternalForce="0,0,0") -> (Pose) {
    angularDampingOverride=0.9999
    angularSpringConstant=200
    bAngularSpring=true
    bOverrideAngularDamping=true
    bUseGravityOverride=true
    simulationSpace=BoneRelative
    boundBone {boneName=front_stop_2}
    componentAppliedLinearAccClamp=<4000 4000 4000>
    componentLinearAccScale=<0.4 0.3 0.3>
    componentLinearVelScale=<0.3 0.3 0.3>
    gravityOverride {z=-650}
    physicsBodyDefinitions [{
      boundBone {boneName=front_stop_2}
      boxExtents=<10 10 10>
      constraintSetup {angularLimitsMax=<0 0 50>, angularLimitsMin=<0 0 0>, angularTarget=<1 0 0>}
      localJointOffset=<1 0 0>
     }]
    relativeSpaceBone {boneName=root}
   }
   n3 = Anim Dynamics - Bone: gun_1(ComponentPose=@n2, Alpha=1, ExternalForce="0,0,0") {
    angularDampingOverride=1
    angularSpringConstant=300
    bLinearSpring=true
    bOverrideAngularDamping=true
    bOverrideLinearDamping=true
    bUseGravityOverride=true
    linearDampingOverride=0.99
    simulationSpace=BoneRelative
    boundBone {boneName=gun_1}
    componentAppliedLinearAccClamp=<4000 4000 4000>
    componentLinearAccScale=<0.3 0.3 0.28>
    componentLinearVelScale=<0.2 0.2 0.22>
    gravityOverride {z=-980}
    physicsBodyDefinitions [{
      boundBone {boneName=gun_1}
      boxExtents=<10 10 10>
      constraintSetup {angularTargetAxis=AxisY, angularLimitsMax=<0 5 15>, angularLimitsMin=<0 -5 -7>, angularTarget=<1 0 0>, linearAxesMax=<1.5 0 0>}
      localJointOffset=<-3 0 0>
     }]
    relativeSpaceBone {boneName=root}
    simSpaceSettings {maxAngularAcceleration=15, maxAngularVelocity=5}
   }
   n4 = Anim Dynamics - Bone: gun_2(ComponentPose=@n3, Alpha=1, ExternalForce="0,0,0") {
    angularDampingOverride=1
    angularSpringConstant=300
    bLinearSpring=true
    bOverrideAngularDamping=true
    bOverrideLinearDamping=true
    bUseGravityOverride=true
    linearDampingOverride=0.999
    simulationSpace=BoneRelative
    boundBone {boneName=gun_2}
    componentAppliedLinearAccClamp=<4000 4000 4000>
    componentLinearAccScale=<0.15 0.2 0.295>
    componentLinearVelScale=<0.2 0.2 0.22>
    gravityOverride {z=-980}
    physicsBodyDefinitions [{
      boundBone {boneName=gun_2}
      boxExtents=<10 10 10>
      constraintSetup {angularTargetAxis=AxisY, angularLimitsMax=<0 7 10>, angularLimitsMin=<0 -3 0>, angularTarget=<1 0 0>, linearAxesMax=<2 0 0>}
      localJointOffset=<-1 0 0>
     }]
    relativeSpaceBone {boneName=root}
    simSpaceSettings {maxAngularAcceleration=15, maxAngularVelocity=5}
   }
   n5 = Anim Dynamics - Bone: ammo(ComponentPose=@n4, Alpha=1, ExternalForce="0,0,0") {
    angularDampingOverride=0.9999
    angularSpringConstant=300
    bOverrideAngularDamping=true
    bUseGravityOverride=true
    simulationSpace=BoneRelative
    boundBone {boneName=ammo}
    componentAppliedLinearAccClamp=<4000 4000 4000>
    componentLinearAccScale=<0.15 0.25 0.31>
    componentLinearVelScale=<0.2 0.25 0.22>
    gravityOverride {z=-980}
    physicsBodyDefinitions [{
      boundBone {boneName=ammo}
      boxExtents=<10 10 10>
      constraintSetup {
       angularTargetAxis=AxisY
       angularLimitsMax=<0 5 0>
       angularLimitsMin=<0 0 0>
       angularTarget=<1 0 0>
       linearAxesMax=<0 0 2>
       linearAxesMin=<0 0 -1>
      }
      localJointOffset=<1.5 0 0>
     }]
    relativeSpaceBone {boneName=root}
    simSpaceSettings {maxAngularAcceleration=15, maxAngularVelocity=5}
   }
   n6 = Component To Local(ComponentPose=@n5)
   n7 = Anim Dynamics - Bone: front_stop_1(ComponentPose=Local To Component(LocalPose=Use cached pose 'Sky/level'()), Alpha=1, GravityOverride="0,0,-650", ExternalForce="0,0,0") {
    angularDampingOverride=0.9999
    angularSpringConstant=200
    bAngularSpring=true
    bOverrideAngularDamping=true
    bUseGravityOverride=true
    simulationSpace=BoneRelative
    boundBone {boneName=front_stop_1}
    componentAppliedLinearAccClamp=<4000 4000 4000>
    componentLinearAccScale=<0.7 0.4 0.26>
    componentLinearVelScale=<0.3 0.3 0.3>
    gravityOverride {z=-650}
    physicsBodyDefinitions [{
      boundBone {boneName=front_stop_1}
      boxExtents=<8 8 8>
      constraintSetup {angularLimitsMax=<0 5 90>, angularLimitsMin=<0 -5 0>, angularTarget=<1 0 0>}
      localJointOffset=<0.8 0 0>
     }]
    relativeSpaceBone {boneName=root}
   }
   n8 = Anim Dynamics - Bone: front_stop_2(ComponentPose=@n7, Alpha=1, GravityOverride="0,0,-650", ExternalForce="0,0,0") {
    angularDampingOverride=0.9999
    angularSpringConstant=200
    bAngularSpring=true
    bOverrideAngularDamping=true
    bUseGravityOverride=true
    simulationSpace=BoneRelative
    boundBone {boneName=front_stop_2}
    componentAppliedLinearAccClamp=<4000 4000 4000>
    componentLinearAccScale=<0.4 0.3 0.3>
    componentLinearVelScale=<0.3 0.3 0.3>
    gravityOverride {z=-650}
    physicsBodyDefinitions [{
      boundBone {boneName=front_stop_2}
      boxExtents=<10 10 10>
      constraintSetup {angularLimitsMax=<0 0 50>, angularLimitsMin=<0 0 0>, angularTarget=<1 0 0>}
      localJointOffset=<1 0 0>
     }]
    relativeSpaceBone {boneName=root}
   }
   n9 = Anim Dynamics - Bone: gun_1(ComponentPose=@n8, Alpha=1, ExternalForce="0,0,0") {
    angularDampingOverride=1
    angularSpringConstant=300
    bLinearSpring=true
    bOverrideAngularDamping=true
    bOverrideLinearDamping=true
    bUseGravityOverride=true
    linearDampingOverride=0.99
    simulationSpace=BoneRelative
    boundBone {boneName=gun_1}
    componentAppliedLinearAccClamp=<4000 4000 4000>
    componentLinearAccScale=<0.3 0.3 0.28>
    componentLinearVelScale=<0.2 0.2 0.22>
    gravityOverride {z=-980}
    physicsBodyDefinitions [{
      boundBone {boneName=gun_1}
      boxExtents=<15 15 35>
      constraintSetup {angularTargetAxis=AxisY, angularLimitsMax=<0 5 15>, angularLimitsMin=<0 -5 -7>, angularTarget=<1 0 0>, linearAxesMax=<1.5 0 0>}
      localJointOffset=<-0.5 0 0>
     }]
    relativeSpaceBone {boneName=root}
    simSpaceSettings {maxAngularAcceleration=15, maxAngularVelocity=5}
   }
   n10 = Anim Dynamics - Bone: gun_2(ComponentPose=@n9, Alpha=1, ExternalForce="0,0,0") {
    angularDampingOverride=1
    angularSpringConstant=300
    bLinearSpring=true
    bOverrideAngularDamping=true
    bOverrideLinearDamping=true
    bUseGravityOverride=true
    linearDampingOverride=0.999
    simulationSpace=BoneRelative
    boundBone {boneName=gun_2}
    componentAppliedLinearAccClamp=<4000 4000 4000>
    componentLinearAccScale=<0.15 0.2 0.295>
    componentLinearVelScale=<0.2 0.2 0.22>
    gravityOverride {z=-980}
    physicsBodyDefinitions [{
      boundBone {boneName=gun_2}
      boxExtents=<15 15 35>
      constraintSetup {angularTargetAxis=AxisY, angularLimitsMax=<0 7 10>, angularLimitsMin=<0 -3 0>, angularTarget=<1 0 0>, linearAxesMax=<2 0 0>}
      localJointOffset=<-0.5 0 0>
     }]
    relativeSpaceBone {boneName=root}
    simSpaceSettings {maxAngularAcceleration=15, maxAngularVelocity=5}
   }
   n11 = Anim Dynamics - Bone: ammo(ComponentPose=@n10, Alpha=1, ExternalForce="0,0,0") {
    angularDampingOverride=0.9999
    angularSpringConstant=300
    bOverrideAngularDamping=true
    bUseGravityOverride=true
    simulationSpace=BoneRelative
    boundBone {boneName=ammo}
    componentAppliedLinearAccClamp=<4000 4000 4000>
    componentLinearAccScale=<0.15 0.25 0.31>
    componentLinearVelScale=<0.2 0.25 0.22>
    gravityOverride {z=-980}
    physicsBodyDefinitions [{
      boundBone {boneName=ammo}
      boxExtents=<10 10 10>
      constraintSetup {
       angularTargetAxis=AxisY
       angularLimitsMax=<0 5 0>
       angularLimitsMin=<0 0 0>
       angularTarget=<1 0 0>
       linearAxesMax=<0 0 2>
       linearAxesMin=<0 0 -1>
      }
      localJointOffset=<1.5 0 0>
     }]
    relativeSpaceBone {boneName=root}
    simSpaceSettings {maxAngularAcceleration=15, maxAngularVelocity=5}
   }
   n12 = Component To Local(ComponentPose=@n11)
   n13 = Get isfullskydivie() -> (isfullskydivie:bool)
   n14 = Anim Dynamics - Bone: front_stop_1(ComponentPose=Local To Component(LocalPose=Use cached pose 'Sky/level'()), Alpha=1, GravityOverride="0,0,-650", ExternalForce="0,0,0") {
    angularDampingOverride=1
    angularSpringConstant=200
    bAngularSpring=true
    bOverrideAngularDamping=true
    bUseGravityOverride=true
    simulationSpace=BoneRelative
    boundBone {boneName=front_stop_1}
    componentAppliedLinearAccClamp=<4000 4000 4000>
    componentLinearAccScale=<0.31 0.4 0.26>
    componentLinearVelScale=<0.3 0.3 0.3>
    gravityOverride {z=-650}
    physicsBodyDefinitions [{
      boundBone {boneName=front_stop_1}
      boxExtents=<8 8 8>
      constraintSetup {angularLimitsMax=<0 5 90>, angularLimitsMin=<0 -5 0>, angularTarget=<1 0 0>}
      localJointOffset=<0.8 0 0>
     }]
    relativeSpaceBone {boneName=root}
   }
   n15 = Anim Dynamics - Bone: front_stop_2(ComponentPose=@n14, Alpha=1, GravityOverride="0,0,-650", ExternalForce="0,0,0") {
    angularDampingOverride=1
    angularSpringConstant=200
    bAngularSpring=true
    bOverrideAngularDamping=true
    bUseGravityOverride=true
    simulationSpace=BoneRelative
    boundBone {boneName=front_stop_2}
    componentAppliedLinearAccClamp=<4000 4000 4000>
    componentLinearAccScale=<0.31 0.3 0.3>
    componentLinearVelScale=<0.3 0.3 0.3>
    gravityOverride {z=-650}
    physicsBodyDefinitions [{
      boundBone {boneName=front_stop_2}
      boxExtents=<10 10 10>
      constraintSetup {angularLimitsMax=<0 0 50>, angularLimitsMin=<0 0 0>, angularTarget=<1 0 0>}
      localJointOffset=<1 0 0>
     }]
    relativeSpaceBone {boneName=root}
   }
   n16 = Anim Dynamics - Bone: gun_1(ComponentPose=@n15, Alpha=1, ExternalForce="0,0,0") {
    angularDampingOverride=1
    angularSpringConstant=300
    bLinearSpring=true
    bOverrideAngularDamping=true
    bOverrideLinearDamping=true
    bUseGravityOverride=true
    linearDampingOverride=0.999
    simulationSpace=BoneRelative
    boundBone {boneName=gun_1}
    componentAppliedLinearAccClamp=<4000 4000 4000>
    componentLinearAccScale=<0.32 0.3 0.28>
    componentLinearVelScale=<0.2 0.2 0.22>
    gravityOverride {z=-980}
    physicsBodyDefinitions [{
      boundBone {boneName=gun_1}
      boxExtents=<15 15 35>
      constraintSetup {angularTargetAxis=AxisY, angularLimitsMax=<0 5 0>, angularLimitsMin=<0 -5 -7>, angularTarget=<1 0 0>, linearAxesMax=<1.5 0 0>}
      localJointOffset=<-1 0 0>
     }]
    relativeSpaceBone {boneName=root}
    simSpaceSettings {maxAngularAcceleration=15, maxAngularVelocity=5}
   }
   n17 = Anim Dynamics - Bone: gun_2(ComponentPose=@n16, Alpha=0.8, ExternalForce="0,0,0") {
    angularDampingOverride=1
    angularSpringConstant=300
    bOverrideAngularDamping=true
    bOverrideLinearDamping=true
    bUseGravityOverride=true
    linearDampingOverride=0.999
    simulationSpace=BoneRelative
    boundBone {boneName=gun_2}
    componentAppliedLinearAccClamp=<4000 4000 4000>
    componentLinearAccScale=<0.32 0.2 0.295>
    componentLinearVelScale=<0.2 0.2 0.22>
    gravityOverride {z=-980}
    physicsBodyDefinitions [{
      boundBone {boneName=gun_2}
      boxExtents=<15 15 35>
      constraintSetup {angularTargetAxis=AxisY, angularLimitsMax=<0 7 10>, angularLimitsMin=<0 -3 0>, angularTarget=<1 0 0>, linearAxesMax=<2 0 0>}
      localJointOffset=<-1 0 0>
     }]
    relativeSpaceBone {boneName=root}
    simSpaceSettings {maxAngularAcceleration=15, maxAngularVelocity=5}
   }
   n18 = Anim Dynamics - Bone: ammo(ComponentPose=@n17, Alpha=1, ExternalForce="0,0,0") {
    angularDampingOverride=1
    angularSpringConstant=300
    bOverrideAngularDamping=true
    bUseGravityOverride=true
    simulationSpace=BoneRelative
    boundBone {boneName=ammo}
    componentAppliedLinearAccClamp=<4000 4000 4000>
    componentLinearAccScale=<0.15 0.25 0.31>
    componentLinearVelScale=<0.2 0.25 0.22>
    gravityOverride {z=-980}
    physicsBodyDefinitions [{
      boundBone {boneName=ammo}
      boxExtents=<20 20 20>
      constraintSetup {
       angularTargetAxis=AxisY
       angularLimitsMax=<0 5 0>
       angularLimitsMin=<0 0 0>
       angularTarget=<1 0 0>
       linearAxesMax=<0 0 2>
       linearAxesMin=<0 0 -1>
      }
      localJointOffset=<1 0 0>
     }]
    relativeSpaceBone {boneName=root}
    simSpaceSettings {maxAngularAcceleration=15, maxAngularVelocity=5}
   }
   Output Pose(Result=Blend Poses by bool(
     BlendPose_0=Component To Local(ComponentPose=@n18)
     BlendPose_1=Blend Poses by bool(BlendPose_0=@n12, BlendPose_1=@n6, bActiveValue=@n13)
     bActiveValue=Get bIsInFrontEnd()
    )) {layerGroup=None}
   n19 = "Transform (Modify) Bone - Bone: pack"(ComponentPose=Local To Component(LocalPose=Input Pose()), Translation="0,0,10", Rotation="0,0,0") {translationMode=BMM_Additive, alphaBoolBlend {blendOption=Linear}, boneToModify {boneName=pack}, translation {z=10}}
   n20 = "Transform (Modify) Bone - Bone: pack"(ComponentPose=@n19, Translation=Get backpack_translation(), Rotation=Get backpack_rotate()) {rotationMode=BMM_Additive, translationMode=BMM_Additive, alphaBoolBlend {blendOption=Linear}, boneToModify {boneName=pack}}
   Save cached pose 'Sky/level'(Pose=Component To Local(ComponentPose=@n20)) {cacheName=Sky/level}
  ]}
 EventGraph {nodes [
   Event BlueprintUpdateAnimation()
   > Set isfullskydivie(isfullskydivie:bool=OR Boolean!(A=Get bIsSkyDiving(), B=Get bIsSkydiveDiveMode(), C:bool=Get bIsSkydiveFloating(), D:bool=Get bIsParachuteOpened()) -> (ReturnValue)) -> (Output_Get:bool)
   > n1 = "Gravity Override State Function (Parameter Asset) (Nativized)"!(
    execute
    CharacterPartAnimInstance=Self-Reference()
    DebugDraw
    GameTimeInSeconds=GetGameTimeInSeconds()
    DebugJointName
    StateLogicLayerName
    Parameters=&/M_MED_FluidWedge_Pack_Parameters
   ) -> (
    then
    OutGravityOverride
    OutCurrentState
    OutGravityOverrideParameters
    OutRigidBodyAnimNodeParameters_ComponentLinearAccScale:Vector=~
    OutRigidBodyAnimNodeParameters_ComponentLinearVelScale:Vector=~
    OutRigidBodyAnimNodeParameters_SimSpaceSettings:Sim Space Settings Structure
   )
   > Set GravityOverride(GravityOverride=@n1)
   > Set Sim Space Settings(Sim Space Settings:Sim Space Settings Structure=@n1.OutRigidBodyAnimNodeParameters_SimSpaceSettings) -> (Output_Get:Sim Space Settings Structure)
   n2 = GetOwnerSkeletalMeshComponent()
   n3 = BreakRotator(InRot=BreakTransform(InTransform=GetDeltaTransformFromRefPose(self=@n2, BoneName=spine_05, BaseName=spine_01)).Rotation)
   n4 = MapRangeClamped(Value=@n3.Yaw, InRangeA=-10, InRangeB=-50, OutRangeB=55)
   > Set backpack_rotate(backpack_rotate=MakeRotator(Roll=@n4, Pitch=float / float(A=@n4, B=7), Yaw=3))
   > Set backpack_translation(backpack_translation=MakeVector(Y=MapRangeClamped(Value=@n3.Yaw, InRangeA=-15, InRangeB=-55, OutRangeB=-7), Z=-8))
   TryGetPawnOwner()
  ]}
]
variables [
 backpack_rotate {type=Rotator}
 backpack_translation {type=Vector}
 GravityOverride {type=Vector}
 isfullskydivie {type=boolean}
 Sim Space Settings {
  type=SimSpaceSettings
  default {
   dampingAlpha=1
   maxAngularAcceleration=10000
   maxAngularVelocity=10000
   maxLinearAcceleration=10000
   maxLinearVelocity=10000
   velocityScaleZ=1
   worldAlpha=0
   externalLinearDragV=<0 0 0>
  }
 }
]
