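# Physics asset for the FluidWedge back bling mesh in BRCosmetics, giving its bones collision shapes and joints for physics simulation.
# One large capsule covers the body, and small capsules sit on a three-bone front_stop chain.
# Constraints link body to MSTR_front_stop, then front_stop_1, then front_stop_2.
# All swing and twist axes are free, so the stored angular limits do not restrict rotation.
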PhysicsAsset /BRCosmetics/Accessories/FORT_Backpacks/Backpack_M_MED_FluidWedge/Meshes/M_MED_FluidWedge_Pack_Physics
bodies {
 MSTR_front_stop [{shapeType=Capsule, length=0.505, radius=0.505, shapeName=MSTR_front_stop_capsule, rotation=<rot 0 -90 0>}]
 body [{shapeType=Capsule, length=18.2981, radius=11.0891, center=<0.6289 2.8001 5.1108>, rotation=<rot 0 -91.4312 -72.1735>}]
 front_stop_1 [{shapeType=Capsule, length=2, radius=2, shapeName=front_stop_1_capsule, center=<2.4767 2.0116 0>, rotation=<rot 90 -0.0002 0>}]
 front_stop_2 [{shapeType=Capsule, length=2, radius=2, shapeName=front_stop_2_capsule, center=<2.2447 1.62 0>, rotation=<rot 90 -21.1883 0>}]
}
constraints [
 {
  bone1Name=front_stop_1
  bone2Name=MSTR_front_stop
  swing1LimitDegrees=45
  swing1Motion=Free
  swing2LimitDegrees=45
  swing2Motion=Free
  twistLimitDegrees=45
  twistMotion=Free
 }
 {
  bone1Name=front_stop_2
  bone2Name=front_stop_1
  swing1LimitDegrees=45
  swing1Motion=Free
  swing2LimitDegrees=45
  swing2Motion=Free
  twistLimitDegrees=45
  twistMotion=Free
 }
 {
  bone1Name=MSTR_front_stop
  bone2Name=body
  swing1LimitDegrees=45
  swing1Motion=Free
  swing2LimitDegrees=45
  swing2Motion=Free
  twistLimitDegrees=45
  twistMotion=Free
 }
]
